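# Sound cue for the insect swarm sting of the Chrono Swarm Grenade, in the projectile effects sound class.
# Each branch mixes sting, bee and hiss layers, with random wave choice and pitch and volume modulation per layer and on the mix.
# A quality node gives the first two levels the full variation pools and the last two only two waves per layer.
# Overrides add short-range occluded attenuation, limit concurrency and exempt random branches from culling.
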SoundCue /ChronoSwarmGrenade/Sounds/Assets/InsectSwarm_Sting_Cue
root=QualityLevel_0
nodes [
 {id=Mixer_0, children [Modulator_0,Modulator_2,Modulator_3], properties {inputVolume [0.5,1.1,0.6]}}
 {id=Mixer_1, children [Modulator_5,Modulator_6,Modulator_7], properties {inputVolume [1,1,0.6]}}
 {id=Modulator_0, children [Random_1], properties {pitchMax=1.1, pitchMin=1, volumeMax=1.1, volumeMin=0.9}}
 {id=Modulator_2, children [Random_2], properties {pitchMax=1.1, pitchMin=0.9, volumeMax=1.1, volumeMin=0.9}}
 {id=Modulator_3, children [Random_5], properties {pitchMax=1.1, pitchMin=0.9, volumeMax=1.1, volumeMin=0.9}}
 {id=Modulator_4, children [Mixer_0], properties {pitchMax=1.1, pitchMin=0.9, volumeMax=1.1, volumeMin=0.9}}
 {id=Modulator_5, children [Random_6], properties {pitchMax=1.1, pitchMin=1, volumeMax=1.1, volumeMin=0.9}}
 {id=Modulator_6, children [Random_7], properties {pitchMax=1.1, pitchMin=0.9, volumeMax=1.1, volumeMin=0.9}}
 {id=Modulator_7, children [Random_8], properties {pitchMax=1.1, pitchMin=0.9, volumeMax=1.1, volumeMin=0.9}}
 {id=Modulator_8, children [Mixer_1], properties {pitchMax=1.1, pitchMin=0.9, volumeMax=1.1, volumeMin=0.9}}
 {id=QualityLevel_0, children [Modulator_4,Modulator_4,Modulator_8,Modulator_8]}
 {id=Random_1, children [WavePlayer_3,WavePlayer_4,WavePlayer_5,WavePlayer_6,WavePlayer_7,WavePlayer_8], properties {weights [1,1,1,1,1,1]}}
 {
  id=Random_2
  children [WavePlayer_9, WavePlayer_10, WavePlayer_11, WavePlayer_12, WavePlayer_13, WavePlayer_14, WavePlayer_15, WavePlayer_16, WavePlayer_19]
  properties {weights [1,1,1,1,1,1,1,1,1]}
 }
 {id=Random_5, children [WavePlayer_28, WavePlayer_29, WavePlayer_30, WavePlayer_31], properties {weights [1,1,1,1]}}
 {id=Random_6, children [WavePlayer_32, WavePlayer_33], properties {weights [1,1]}}
 {id=Random_7, children [WavePlayer_38, WavePlayer_39], properties {weights [1,1]}}
 {id=Random_8, children [WavePlayer_47, WavePlayer_49], properties {weights [1,1]}}
 {id=WavePlayer_10, properties {soundWaveAssetPtr=&/InsectSwarm_Sting_BeeLayer_02}}
 {id=WavePlayer_11, properties {soundWaveAssetPtr=&/InsectSwarm_Sting_BeeLayer_03}}
 {id=WavePlayer_12, properties {soundWaveAssetPtr=&/InsectSwarm_Sting_BeeLayer_04}}
 {id=WavePlayer_13, properties {soundWaveAssetPtr=&/InsectSwarm_Sting_BeeLayer_05}}
 {id=WavePlayer_14, properties {soundWaveAssetPtr=&/InsectSwarm_Sting_BeeLayer_06}}
 {id=WavePlayer_15, properties {soundWaveAssetPtr=&/InsectSwarm_Sting_BeeLayer_07}}
 {id=WavePlayer_16, properties {soundWaveAssetPtr=&/InsectSwarm_Sting_BeeLayer_08}}
 {id=WavePlayer_19, properties {soundWaveAssetPtr=&/InsectSwarm_Sting_BeeLayer_09}}
 {id=WavePlayer_28, properties {soundWaveAssetPtr=&/InsectSwarm_Sting_HissLayer_01}}
 {id=WavePlayer_29, properties {soundWaveAssetPtr=&/InsectSwarm_Sting_HissLayer_02}}
 {id=WavePlayer_3, properties {soundWaveAssetPtr=&/InsectSwarm_Sting_01}}
 {id=WavePlayer_30, properties {soundWaveAssetPtr=&/InsectSwarm_Sting_HissLayer_03}}
 {id=WavePlayer_31, properties {soundWaveAssetPtr=&/InsectSwarm_Sting_HissLayer_04}}
 {id=WavePlayer_32, properties {soundWaveAssetPtr=&InsectSwarm_Sting_01}}
 {id=WavePlayer_33, properties {soundWaveAssetPtr=&/InsectSwarm_Sting_02}}
 {id=WavePlayer_38, properties {soundWaveAssetPtr=&/InsectSwarm_Sting_BeeLayer_01}}
 {id=WavePlayer_39, properties {soundWaveAssetPtr=&InsectSwarm_Sting_BeeLayer_02}}
 {id=WavePlayer_4, properties {soundWaveAssetPtr=&InsectSwarm_Sting_02}}
 {id=WavePlayer_47, properties {soundWaveAssetPtr=&InsectSwarm_Sting_HissLayer_01}}
 {id=WavePlayer_49, properties {soundWaveAssetPtr=&InsectSwarm_Sting_HissLayer_03}}
 {id=WavePlayer_5, properties {soundWaveAssetPtr=&/InsectSwarm_Sting_03}}
 {id=WavePlayer_6, properties {soundWaveAssetPtr=&/InsectSwarm_Sting_04}}
 {id=WavePlayer_7, properties {soundWaveAssetPtr=&/InsectSwarm_Sting_05}}
 {id=WavePlayer_8, properties {soundWaveAssetPtr=&/InsectSwarm_Sting_06}}
 {id=WavePlayer_9, properties {soundWaveAssetPtr=&InsectSwarm_Sting_BeeLayer_01}}
]
properties {
 bExcludeFromRandomNodeBranchCulling=true
 bOverrideAttenuation=true
 bOverrideConcurrency=true
 maxDistance=900
 soundClassObject=&Fort_WeaponRanged_Projectile_Effects
 volumeMultiplier=0.4
 attenuationOverrides {
  bEnableOcclusion=true
  falloffDistance=800
  occlusionInterpolationTime=0.5
  occlusionLowPassFilterFrequency=2000
  occlusionTraceChannel=ECC_GameTraceChannel11
  occlusionVolumeAttenuation=0.4
  attenuationShapeExtents {x=100}
 }
 concurrencyOverrides {maxCount=4}
}
