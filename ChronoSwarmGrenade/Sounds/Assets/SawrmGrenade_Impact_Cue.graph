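# Sound cue for the Chrono Swarm Grenade impact, built from insect-swarm jar impact recordings.
# After a short delay it mixes close, medium and distant layers, each with its own attenuation, sound class and randomized waves.
# The close layer blends a jar impact with a delayed, higher-pitched bee burst, while the distant layer plays quieter and lower-pitched.
# A concurrency override caps simultaneous instances.
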SoundCue /ChronoSwarmGrenade/Sounds/Assets/SawrmGrenade_Impact_Cue
root=Delay_0
nodes [
 {
  id=Attenuation_0
  children [Modulator_1]
  properties {
   attenuationSettings=/Game/Sounds/Attenuation/Fort_Explosion_Close_Athena_Att
   attenuationOverrides {falloffDistance=900, attenuationShapeExtents {x=100}}
  }
 }
 {id=Attenuation_1, children [Modulator_6], properties {attenuationSettings=/Game/Sounds/Attenuation/Fort_Explosion_Medium_Athena_Att}}
 {
  id=Attenuation_2
  children [Modulator_8]
  properties {
   attenuationSettings=/Game/Sounds/Attenuation/Fort_Explosion_Distant_Athena_Att
   attenuationOverrides {falloffDistance=3000, attenuationShapeExtents {x=1000}}
  }
 }
 {id=Delay_0, children [Mixer_5], properties {delayMax=0.05, delayMin=0.05}}
 {id=Delay_1, children [Random_7], properties {delayMax=0.2, delayMin=0.2}}
 {id=Mixer_3, children [Random_6,Modulator_2], properties {inputVolume [2.3,0.6]}}
 {id=Mixer_5, children [SoundClass_1,SoundClass_4,SoundClass_5], properties {inputVolume [0.5,0.35,0.35]}}
 {id=Modulator_1, children [Mixer_3], properties {pitchMax=1.1, pitchMin=0.9, volumeMax=1.1, volumeMin=0.9}}
 {id=Modulator_2, children [Delay_1], properties {pitchMax=1.1, pitchMin=1.1}}
 {id=Modulator_6, children [Random_8], properties {pitchMax=1.1, pitchMin=0.9, volumeMax=1.3, volumeMin=1.2}}
 {id=Modulator_8, children [Random_9], properties {pitchMax=0.9, pitchMin=0.7, volumeMax=0.45, volumeMin=0.35}}
 {id=Random_6, children [WavePlayer_17, WavePlayer_18, WavePlayer_24], properties {weights [1,1,1]}}
 {id=Random_7, children [WavePlayer_25, WavePlayer_26, WavePlayer_27], properties {weights [1,1,1]}}
 {id=Random_8, children [WavePlayer_28, WavePlayer_29, WavePlayer_30], properties {weights [1,1,1]}}
 {id=Random_9, children [WavePlayer_31, WavePlayer_32, WavePlayer_33], properties {weights [1,1,1]}}
 {
  id=SoundClass_1
  children [Attenuation_0]
  properties {soundClassOverride=&Fort_WeaponRanged_Fire_3P_PSM}
 }
 {
  id=SoundClass_4
  children [Attenuation_1]
  properties {soundClassOverride=&Fort_WeaponRanged_Fire_Distant}
 }
 {
  id=SoundClass_5
  children [Attenuation_2]
  properties {soundClassOverride=&Fort_WeaponRanged_Fire_Distant}
 }
 {id=WavePlayer_17, properties {soundWaveAssetPtr=&/InsectSwarm_Jar_Impact_01_Jar}}
 {id=WavePlayer_18, properties {soundWaveAssetPtr=&/InsectSwarm_Jar_Impact_02_Jar}}
 {id=WavePlayer_24, properties {soundWaveAssetPtr=&/InsectSwarm_Jar_Impact_03_Jar}}
 {id=WavePlayer_25, properties {soundWaveAssetPtr=&/InsectSwarm_Jar_Impact_Bees_01}}
 {id=WavePlayer_26, properties {soundWaveAssetPtr=&/InsectSwarm_Jar_Impact_Bees_02}}
 {id=WavePlayer_27, properties {soundWaveAssetPtr=&/InsectSwarm_Jar_Impact_Bees_03}}
 {id=WavePlayer_28, properties {soundWaveAssetPtr=&/InsectSwarm_Jar_Impact_Medium_01}}
 {id=WavePlayer_29, properties {soundWaveAssetPtr=&/InsectSwarm_Jar_Impact_Medium_02}}
 {id=WavePlayer_30, properties {soundWaveAssetPtr=&/InsectSwarm_Jar_Impact_Medium_03}}
 {id=WavePlayer_31, properties {soundWaveAssetPtr=&/InsectSwarm_Jar_Impact_Distant_01}}
 {id=WavePlayer_32, properties {soundWaveAssetPtr=&/InsectSwarm_Jar_Impact_Distant_02}}
 {id=WavePlayer_33, properties {soundWaveAssetPtr=&/InsectSwarm_Jar_Impact_Distant_03}}
]
properties {bOverrideConcurrency=true, maxDistance=27500, volumeMultiplier=0.7, concurrencyOverrides {maxCount=4}}
