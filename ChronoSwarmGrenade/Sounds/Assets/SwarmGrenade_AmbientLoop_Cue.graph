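# Ambient looping sound cue for the Chrono Swarm grenade, built from insect-swarm-in-jar loops.
# A Random node picks one of two equally weighted looping waves, which feeds a Modulator that varies pitch.
# The cue plays quiet and slightly lowered in pitch, uses its own attenuation asset and the gadget items sound class, and overrides concurrency to cap simultaneous instances.
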SoundCue /ChronoSwarmGrenade/Sounds/Assets/SwarmGrenade_AmbientLoop_Cue
root=Modulator_2
nodes [
 {id=Modulator_2, children [Random_2], properties {pitchMax=1.1}}
 {id=Random_2, children [WavePlayer_8, WavePlayer_10], properties {weights [1,1]}}
 {id=WavePlayer_10, properties {bLooping=true, soundWaveAssetPtr=&/InsectSwarm_Jar_AmbientLoop_03}}
 {id=WavePlayer_8, properties {bLooping=true, soundWaveAssetPtr=&/InsectSwarm_Jar_AmbientLoop_01}}
]
properties {
 attenuationSettings=/ChronoSwarmGrenade/Sounds/SwarmGrenade_AmbientLoop_Att
 bOverrideConcurrency=true
 maxDistance=400
 pitchMultiplier=0.9
 soundClassObject=&Fort_Gadgets_Items
 volumeMultiplier=0.3
 concurrencyOverrides {maxCount=2}
}
