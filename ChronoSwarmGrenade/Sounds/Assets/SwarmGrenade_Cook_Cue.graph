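# Sound cue for cooking the Chrono Swarm Grenade, playing an insect swarm jar sound.
# A Random node picks between two equally weighted wave variations, played quieter and slightly pitched up.
# It uses gadget-placement attenuation and the gadget items sound class, with a concurrency override capping simultaneous instances.
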SoundCue /ChronoSwarmGrenade/Sounds/Assets/SwarmGrenade_Cook_Cue
root=Random_0
nodes [
 {id=Random_0, children [WavePlayer_1,WavePlayer_2], properties {weights [1,1]}}
 {id=WavePlayer_1, properties {soundWaveAssetPtr=&/InsectSwarm_Jar_Cook_01}}
 {id=WavePlayer_2, properties {soundWaveAssetPtr=&/InsectSwarm_Jar_Cook_02}}
]
properties {
 attenuationSettings=/Game/Sounds/Attenuation/Fort_GadgetPlace_Att
 bOverrideConcurrency=true
 maxDistance=2000
 pitchMultiplier=1.1
 soundClassObject=&Fort_Gadgets_Items
 volumeMultiplier=0.3
 concurrencyOverrides {maxCount=2}
}
