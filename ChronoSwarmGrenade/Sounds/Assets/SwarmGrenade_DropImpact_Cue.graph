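# Sound cue for the Chrono Swarm Grenade's drop impact, playing insect swarm jar drop sounds.
# A Random node picks one of three equally weighted drop waves, and a Modulator above it varies playback for each hit.
# The cue plays at reduced volume and slightly lowered pitch, using the shared weapon drop attenuation and the loot collect sound class.
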SoundCue /ChronoSwarmGrenade/Sounds/Assets/SwarmGrenade_DropImpact_Cue
root=Modulator_0
nodes [
 {id=Modulator_0, children [Random_1], properties {pitchMax=1}}
 {id=Random_1, children [WavePlayer_0,WavePlayer_3,WavePlayer_4], properties {weights [1,1,1]}}
 {id=WavePlayer_0, properties {soundWaveAssetPtr=&/InsectSwarm_Jar_Drop_01}}
 {id=WavePlayer_3, properties {soundWaveAssetPtr=&/InsectSwarm_Jar_Drop_02}}
 {id=WavePlayer_4, properties {soundWaveAssetPtr=&/InsectSwarm_Jar_Drop_03}}
]
properties {
 attenuationSettings=&Fort_Weapon_Drop_Att
 maxDistance=3000
 pitchMultiplier=0.9
 soundClassObject=&Fort_Loot_Collect
 volumeMultiplier=0.27
 attenuationOverrides {falloffDistance=2400, stereoSpread=2000, attenuationShapeExtents {x=800}}
}
