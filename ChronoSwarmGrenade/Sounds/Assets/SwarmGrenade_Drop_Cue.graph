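# Sound cue for dropping the Chrono Swarm Grenade, built on an insect swarm jar drop sound.
# A single wave player feeds a Modulator at default settings, adding slight random pitch and volume variation on each play.
# The cue raises pitch, lowers volume, uses the loot-collect sound class and references shared weapon-drop attenuation, with its own attenuation override values also stored.
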SoundCue /ChronoSwarmGrenade/Sounds/Assets/SwarmGrenade_Drop_Cue
root=Modulator_0
nodes [
 {id=Modulator_0, children [WavePlayer_0]}
 {id=WavePlayer_0, properties {soundWaveAssetPtr=&/InsectSwarm_Jar_Drop_B}}
]
properties {
 attenuationSettings=&Fort_Weapon_Drop_Att
 maxDistance=3000
 pitchMultiplier=1.2
 soundClassObject=&Fort_Loot_Collect
 volumeMultiplier=0.55
 attenuationOverrides {falloffDistance=2400, stereoSpread=2000, attenuationShapeExtents {x=800}}
}
