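# Sound cue for equipping the Chrono Swarm Grenade, playing an insect swarm jar equip sound.
# A Random node picks one of three equip wave variations with equal weight and feeds a Modulator at the root.
# It plays as player foley with generic equip attenuation and a concurrency override that prevents new instances once the limit is reached.
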SoundCue /ChronoSwarmGrenade/Sounds/Assets/SwarmGrenade_Equip_Cue
root=Modulator_1
nodes [
 {id=Modulator_1, children [Random_0]}
 {id=Random_0, children [WavePlayer_0,WavePlayer_1,WavePlayer_3], properties {weights [1,1,1]}}
 {id=WavePlayer_0, properties {soundWaveAssetPtr=&/InsectSwarm_Jar_Equip_01}}
 {id=WavePlayer_1, properties {soundWaveAssetPtr=&/InsectSwarm_Jar_Equip_02}}
 {id=WavePlayer_3, properties {soundWaveAssetPtr=&/InsectSwarm_Jar_Equip_03}}
]
properties {
 attenuationSettings=/Game/Sounds/Attenuation/Fort_Player_Equip_Generic_Att
 bOverrideConcurrency=true
 maxDistance=1800
 soundClassObject=&Fort_Player_Foley
 volumeMultiplier=0.885
 concurrencyOverrides {resolutionRule=PreventNew}
}
