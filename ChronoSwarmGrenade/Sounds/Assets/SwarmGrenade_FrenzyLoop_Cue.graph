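# Sound cue for the Chrono Swarm Grenade's frenzy state, playing a looping insect swarm attack layered with a short attack-start burst.
# A mixer blends two branches, each randomly picking one of three waves through a pitch and volume modulator.
# The frenzy branch loops, while the fast-fade start branch plays once at higher pitch and lower mix level.
# Spatialization references the grenade's attenuation asset, and a concurrency override caps instances, stopping the farthest first.
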SoundCue /ChronoSwarmGrenade/Sounds/Assets/SwarmGrenade_FrenzyLoop_Cue
root=Mixer_1
nodes [
 {id=Mixer_1, children [Modulator_1,Modulator_4], properties {inputVolume [0.8,0.6]}}
 {id=Modulator_1, children [Random_2], properties {pitchMax=1.1, pitchMin=1, volumeMax=1.1, volumeMin=0.9}}
 {id=Modulator_4, children [Random_10], properties {pitchMax=1.4, pitchMin=1.3, volumeMax=1.1, volumeMin=1}}
 {id=Random_10, children [WavePlayer_41, WavePlayer_42, WavePlayer_43], properties {weights [1,1,1]}}
 {id=Random_2, children [WavePlayer_7,WavePlayer_8,WavePlayer_9], properties {weights [1,1,1]}}
 {id=WavePlayer_41, properties {soundWaveAssetPtr=&/InsectSwarm_AtackStart_FastFade_01}}
 {id=WavePlayer_42, properties {soundWaveAssetPtr=&/InsectSwarm_AtackStart_FastFade_02}}
 {id=WavePlayer_43, properties {soundWaveAssetPtr=&/InsectSwarm_AtackStart_FastFade_03}}
 {id=WavePlayer_7, properties {bLooping=true, soundWaveAssetPtr=&/InsectSwarm_AtackLoop_Frenzy_01}}
 {id=WavePlayer_8, properties {bLooping=true, soundWaveAssetPtr=&/InsectSwarm_AtackLoop_Frenzy_02}}
 {id=WavePlayer_9, properties {bLooping=true, soundWaveAssetPtr=&/InsectSwarm_AtackLoop_Frenzy_03}}
]
properties {
 attenuationSettings=/ChronoSwarmGrenade/Sounds/SwarmGrenade_Att
 bOverrideConcurrency=true
 maxDistance=3000
 soundClassObject=&Fort_WeaponRanged_Fire_3P_PSM
 volumeMultiplier=0.33
 attenuationOverrides {distanceAlgorithm=NaturalSound, falloffDistance=4000, attenuationShapeExtents {x=1000}}
 concurrencyOverrides {maxCount=3, resolutionRule=StopFarthestThenPreventNew}
}
