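# Sound cue for the looping insect-swarm jar sound of the Chrono Swarm grenade while it is in the air.
# A Random node picks one of two equally weighted looping wave players, which feeds a Modulator at default settings for slight variation.
# The cue lowers pitch and volume, uses projectile in-air loop attenuation and the gadgets sound class, and overrides concurrency to allow a single instance.
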SoundCue /ChronoSwarmGrenade/Sounds/Assets/SwarmGrenade_InAirLoop_Cue
root=Modulator_0
nodes [
 {id=Modulator_0, children [Random_0]}
 {id=Random_0, children [WavePlayer_1,WavePlayer_2], properties {weights [1,1]}}
 {id=WavePlayer_1, properties {bLooping=true, soundWaveAssetPtr=&/InsectSwarm_Jar_InAirLoop_01}}
 {id=WavePlayer_2, properties {bLooping=true, soundWaveAssetPtr=&/InsectSwarm_Jar_InAirLoop_02}}
]
properties {
 attenuationSettings=/Game/Sounds/Attenuation/Fort_Weapon_Projectile_InAir_Loop_Att
 bOverrideConcurrency=true
 maxDistance=4500
 pitchMultiplier=0.85
 soundClassObject=&Fort_Gadgets_Items
 volumeMultiplier=0.45
 concurrencyOverrides {maxCount=1}
}
