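# Sound cue for picking up the Chrono Swarm Grenade, playing an insect-swarm jar pickup sound.
# A random node picks one of three equally weighted wave variants, then a modulator processes it at default settings.
# It uses weapon pickup attenuation and the gadgets item sound class at reduced volume.
# A concurrency override prevents new instances from starting once the limit is reached.
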SoundCue /ChronoSwarmGrenade/Sounds/Assets/SwarmGrenade_Pickup_Cue
root=Modulator_0
nodes [
 {id=Modulator_0, children [Random_0]}
 {id=Random_0, children [WavePlayer_1,WavePlayer_2,WavePlayer_3], properties {weights [1,1,1]}}
 {id=WavePlayer_1, properties {soundWaveAssetPtr=&/InsectSwarm_Jar_Pickup_01}}
 {id=WavePlayer_2, properties {soundWaveAssetPtr=&/InsectSwarm_Jar_Pickup_02}}
 {id=WavePlayer_3, properties {soundWaveAssetPtr=&/InsectSwarm_Jar_Pickup_03}}
]
properties {
 attenuationSettings=&Fort_Weapon_Pickup_Att
 bOverrideConcurrency=true
 maxDistance=3500
 soundClassObject=&Fort_Gadgets_Items
 volumeMultiplier=0.26
 concurrencyOverrides {resolutionRule=PreventNew}
}
