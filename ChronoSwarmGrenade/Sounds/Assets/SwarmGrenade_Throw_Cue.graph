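# Sound cue for throwing the Chrono Swarm Grenade, built from insect swarm jar throw samples.
# A Random node picks one of three equally weighted waves, then a Modulator slightly varies pitch and volume.
# It uses gadget-throw attenuation and the gadget items sound class, with a concurrency override that limits simultaneous instances.
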SoundCue /ChronoSwarmGrenade/Sounds/Assets/SwarmGrenade_Throw_Cue
root=Modulator_0
nodes [
 {id=Modulator_0, children [Random_0], properties {pitchMax=1.1, pitchMin=0.9, volumeMax=1.1, volumeMin=0.9}}
 {id=Random_0, children [WavePlayer_1,WavePlayer_2,WavePlayer_3], properties {weights [1,1,1]}}
 {id=WavePlayer_1, properties {soundWaveAssetPtr=&/InsectSwarm_Jar_Throw_01}}
 {id=WavePlayer_2, properties {soundWaveAssetPtr=&/InsectSwarm_Jar_Throw_02}}
 {id=WavePlayer_3, properties {soundWaveAssetPtr=&/InsectSwarm_Jar_Throw_03}}
]
properties {
 attenuationSettings=/Game/Sounds/Attenuation/Fort_GadgetThrow_Att
 bOverrideConcurrency=true
 maxDistance=1536
 soundClassObject=&Fort_Gadgets_Items
 volumeMultiplier=0.8
 concurrencyOverrides {maxCount=2}
}
