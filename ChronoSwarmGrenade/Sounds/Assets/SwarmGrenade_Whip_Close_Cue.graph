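# Sound cue for the close-range whip-by sound of the Chrono Swarm Grenade projectile, built from insect-swarm jar whip recordings.
# A short fixed delay feeds a modulator, which plays one of six evenly weighted random whip variations.
# It borrows the shared bullet-whip attenuation, routes to the ranged projectile effects sound class, and overrides concurrency to a single instance per owner.
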SoundCue /ChronoSwarmGrenade/Sounds/Assets/SwarmGrenade_Whip_Close_Cue
root=Delay_0
nodes [
 {id=Delay_0, children [Modulator_0], properties {delayMax=0.1, delayMin=0.1}}
 {id=Modulator_0, children [Random_1]}
 {id=Random_1, children [WavePlayer_3,WavePlayer_4,WavePlayer_5,WavePlayer_6,WavePlayer_7,WavePlayer_8], properties {weights [1,1,1,1,1,1]}}
 {id=WavePlayer_3, properties {soundWaveAssetPtr=&/InsectSwarm_Jar_Whip_01}}
 {id=WavePlayer_4, properties {soundWaveAssetPtr=&/InsectSwarm_Jar_Whip_02}}
 {id=WavePlayer_5, properties {soundWaveAssetPtr=&/InsectSwarm_Jar_Whip_03}}
 {id=WavePlayer_6, properties {soundWaveAssetPtr=&/InsectSwarm_Jar_Whip_04}}
 {id=WavePlayer_7, properties {soundWaveAssetPtr=&/InsectSwarm_Jar_Whip_05}}
 {id=WavePlayer_8, properties {soundWaveAssetPtr=&/InsectSwarm_Jar_Whip_06}}
]
properties {
 attenuationSettings=/Game/Athena/Sounds/Weapons/BulletWhips/bullet_whip_Att.Bullet_Whip_Att
 bOverrideConcurrency=true
 maxDistance=2097152
 pitchMultiplier=1.1
 soundClassObject=&Fort_WeaponRanged_Projectile_Effects
 volumeMultiplier=0.6
 concurrencyOverrides {bLimitToOwner=true, maxCount=1}
}
